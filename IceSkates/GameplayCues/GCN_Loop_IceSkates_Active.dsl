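; Looping gameplay cue for active Ice Skates, attaching effects to both feet and managing skate audio plus optional hiding while cloaked.
; For the local player, a timer polls item durability, playing a warning and driving crack effects below a threshold.
; Pawn events and tag watchers drive slide loops, collision sounds and, via surface tracking, water slide and splash effects.
; Removal spawns expire effects and cleans up bindings, watchers, audio and the added footstep sound library.
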(event Custom|InitExpirationTimerHandle
  ;[1 1 1] Handle ability expire warning - local player only
  (bind _returnvalue (Utilities|Time|SetTimerbyEvent (delegate Custom|ExpireTickHandle) (Variables|Default|GetExpirationSoundPeriod) true))
  (Variables|Default|SetExpirationTimerHandle _returnvalue)
  (Variables|Default|SetCrack 0.0))

(event Custom|ExpireTickHandle
  ;[1 1 1] Handle ability expire warning - local player only
  (bind _output_get (Variables|Default|SetItemDurabilityNormalized (Item|GetNormalizedDurability (Variables|Default|GetTargetInventoryItem))))
  (if (and (< _output_get (ScalableFloat|GetValueatLevel (Variables|Default|GetExpirationDurabilityThreshold))) (> (ScalableFloat|GetValueatLevel (Variables|Default|GetExpirationDurabilityThreshold)) 0))
    (Audio|SpawnSoundAttached :Sound (Variables|Sound|GetExpire%20Warning%20Sound) :AttachToComponent (Variables|/Script/Engine.Actor|GetRootComponent (Variables|Default|GetTargetPlayer)) :AttachPointName "None" :LocationType "KeepRelativeOffset" :bStopWhenAttachedToDestroyed false :VolumeMultiplier 0.7 :PitchMultiplier 1.2 :StartTime 0.0 :bAutoDestroy true)
    (bind _output_get_2 (Variables|Default|SetCrack (- 1.0 (/ (Variables|Default|GetItemDurabilityNormalized) (Variables|Default|GetExpirationDurabilityThreshold)))))
    (for _array_element (Variables|Default|GetSpawnedEmitters)
      (Effects|Components|ParticleSystem|SetFloatParameter _array_element "Crack" _output_get_2)
      (Effects|Components|ParticleSystem|SetFloatParameter _array_element "Time" (Utilities|Time|GetGameTimeinSeconds 0))
      (Niagara|SpawnSystematLocation 0 "/FNGameplayCues/Athena/Items/Traps/FreezeTrap/FX/NS_Ice_Foot_Crack.NS_Ice_Foot_Crack" (Transformation|GetWorldLocation _array_element) "0.000000,90.000000,0.000000"))))

(event Custom|ClearExpireTimerHandle
  ;[1 1 1] Handle ability expire warning - local player only
  (Utilities|Time|ClearandInvalidateTimerbyHandle 0 (Variables|Default|GetExpirationTimerHandle)))

(event EventOnLoopingStart(EitherVFX-preferthisversion)
  (params (MyTarget :type Actor%20Object%20Reference) (Parameters :type Gameplay%20Cue%20Parameters%20Structure%20%28by%20ref%29) (ParticleComponents :type Array%20of%20FXSystem%20Component%20Object%20References) (AudioComponents :type Array%20of%20Audio%20Component%20Object%20References))
  (bind _spawnedemitters (Variables|Default|GetSpawnedEmitters))
  (bind _output (Utilities|Array|Get(acopy) _spawnedemitters 1))
  (bind _output_2 (Utilities|Array|Get(acopy) _spawnedemitters))
  (if (> (ScalableFloat|GetValueasDouble (Variables|Default|GetHideVisualsDuringInvis)) 0)
    (bind _asynctaskproxy (Ability|Tasks|WaitGameplayTagAddToActor MyTarget (Variables|Default|GetCloakTag))
      (:then
        (Variables|Default|SetTag%20Add%20Watcher _asynctaskproxy)
        (bind _asynctaskproxy_2 (Ability|Tasks|WaitGameplayTagRemoveFromActor MyTarget (Variables|Default|GetCloakTag))
          (:then
            (Variables|Default|SetTag%20Remove%20Watcher _asynctaskproxy_2))
          (:Removed
            (for _array_element ParticleComponents
              (bind _asniagara_particle_system_component (Utilities|Casting|CastToNiagaraComponent _array_element))
              (Rendering|SetVisibility _asniagara_particle_system_component true true)))))
      (:Added
        (for _array_element_2 ParticleComponents
          (bind _asniagara_particle_system_component_2 (Utilities|Casting|CastToNiagaraComponent _array_element_2))
          (Rendering|SetVisibility _asniagara_particle_system_component_2 false true)))))
  (Utilities|FlowControl|Sequence
    (:then_0
      (Variables|Default|SetSpawnedEmitters ParticleComponents)
      (if (Utilities|Array|IsValidIndex AudioComponents)
        (Variables|Default|SetSpawnedAudioComps AudioComponents)
        (CallFunction|SlidingAudio)
        (else
          _))
      (CallFunction|ClearExpireTimerHandle)
      (Variables|Default|SetCrack 0.0))
    (:then_1
      (bind _asfort_player_pawn (Utilities|Casting|CastToFortPlayerPawn MyTarget))
      (bind _output_get (Variables|Default|SetTargetPlayer _asfort_player_pawn))
      (Variables|Default|SetTargetInventoryItem (FortWeapon|GetInventoryItem (Weapon|GetCurrentWeapon _output_get)))
      (Transformation|AttachComponentToComponent _output_2 (Variables|/Script/Engine.Character|GetMesh _asfort_player_pawn) "foot_l" "SnapToTarget" "SnapToTarget" "KeepWorld" true)
      (Effects|Components|ParticleSystem|SetVectorParameter _output_2 "BlockOrientation" "0.000000,0.000000,1.000000")
      (Transformation|AttachComponentToComponent _output (Variables|/Script/Engine.Character|GetMesh _asfort_player_pawn) "foot_r" "SnapToTarget" "SnapToTarget" "KeepWorld" true)
      (Effects|Components|ParticleSystem|SetVectorParameter _output "BlockOrientation" "0.000000,0.000000,-1.000000")
      (Effects|Components|ParticleSystem|SetFloatParameter (Variables|Default|GetSpawnedEmitters) "Time" (Utilities|Time|GetGameTimeinSeconds 0))
      (Utilities|FlowControl|Sequence
        (:then_0
          ;[1 1 1] Handle ability expire warning - local player only
          (if (Pawn|IsLocallyControlled (Variables|Default|GetTargetPlayer))
            (CallFunction|InitExpirationTimerHandle)))
        (:then_1
          ;[1 1 1] Bind footstep and collision events for audio
          (bind _asfort_player_pawn_athena (Utilities|Casting|CastToFortPlayerPawnAthena (Variables|Default|GetTargetPlayer)))
          (Variables|Default|SetAs%20Fort%20Player%20Pawn%20Athena _asfort_player_pawn_athena)
          (EventDispatchers|BindEventtoOnFootstepEvent _asfort_player_pawn_athena (EventDispatchers|CreateEvent :!event Custom|On%20Player%20Step))
          (bind _returnvalue (AddComponent|NODE_AddFortCollisionAudioComponent-1))
          (Variables|Default|SetCollisionAudioComp _returnvalue)
          (Collision|BindEventtoOnActorHit _asfort_player_pawn_athena (EventDispatchers|CreateEvent :!event Custom|On%20Target%20Pawn%20Hit%20Event))
          (Audio|SetCollisionSoundatIndex (Variables|Default|GetCollisionAudioComp) 0 (Variables|Sound|GetCollision%20Sound))
          (EventDispatchers|BindEventtoOnPawnLanded _asfort_player_pawn_athena (EventDispatchers|CreateEvent :!event Custom|On%20Target%20Pawn%20Landed))
          (Character|BindEventtoMovementModeChangedDelegate _asfort_player_pawn_athena (EventDispatchers|CreateEvent :!event Custom|On%20Target%20Pawn%20Movement%20Mode%20Changed))
          (bind _asynctaskproxy_3 (Ability|Tasks|WaitGameplayTagAddToActor (Variables|Default|GetTargetPlayer) "(TagName=\"Gameplay.Action.Player.Sprint\")")
            (:then
              (Variables|Default|SetSprintTagAddWatcher _asynctaskproxy_3)
              (CallFunction|AddSoundLibrary))
            (:Added
              (CallFunction|SlidingAudio))) ; this is used to avoid a delay between stopping the sliding sound\nand the first footstep
        )))
    (:then_2
      (CallFunction|InitializeSliding))))

(event EventOnRemoval(EitherVFX-preferthisversion)
  (params (MyTarget :type Actor%20Object%20Reference) (Parameters :type Gameplay%20Cue%20Parameters%20Structure%20%28by%20ref%29) (ParticleComponents :type Array%20of%20FXSystem%20Component%20Object%20References) (AudioComponents :type Array%20of%20Audio%20Component%20Object%20References) (BurstCameraShakeInstance :type Camera%20Shake%20Base%20Object%20Reference) (BurstDecalInstance :type Decal%20Actor%20Object%20Reference))
  (CallFunction|ClearExpireTimerHandle)
  (bind _asfort_player_pawn (Utilities|Casting|CastToFortPlayerPawn MyTarget))
  (Effects|SpawnFXSystemAttached :FXSystemAsset (Variables|Default|GetExpireFX) :AttachToComponent (Variables|/Script/Engine.Character|GetMesh _asfort_player_pawn) :AttachPointName "foot_l" :Scale "1.000000,1.000000,1.000000" :LocationType "SnapToTarget" :bAutoDestroy true :PoolingMethod "None" :bAutoActivate true)
  (Effects|SpawnFXSystemAttached :FXSystemAsset (Variables|Default|GetExpireFX) :AttachToComponent (Variables|/Script/Engine.Character|GetMesh _asfort_player_pawn) :AttachPointName "foot_r" :Scale "1.000000,1.000000,1.000000" :LocationType "SnapToTarget" :bAutoDestroy true :PoolingMethod "None" :bAutoActivate true)
  ;[1 1 1] Unbind footstep and collision events
  (bind _asfort_player_pawn_athena (Utilities|Casting|CastToFortPlayerPawnAthena MyTarget))
  (EventDispatchers|UnbindEventfromOnFootstepEvent _asfort_player_pawn_athena (EventDispatchers|CreateEvent :!event Custom|On%20Player%20Step))
  (Collision|UnbindEventfromOnActorHit _asfort_player_pawn_athena (EventDispatchers|CreateEvent :!event Custom|On%20Target%20Pawn%20Hit%20Event))
  (EventDispatchers|UnbindEventfromOnPawnLanded _asfort_player_pawn_athena (EventDispatchers|CreateEvent :!event Custom|On%20Target%20Pawn%20Landed))
  (Character|UnbindEventfromMovementModeChangedDelegate _asfort_player_pawn_athena (EventDispatchers|CreateEvent :!event Custom|On%20Target%20Pawn%20Movement%20Mode%20Changed))
  ; Clean up audio
  (CallFunction|RemoveSoundLibrary)
  (bind _slidingaudiocomponent (Variables|Default|GetSlidingAudioComponent)
    (:then
      (Components|DestroyComponent _slidingaudiocomponent self))
    (:else
      _))
  (Components|DestroyComponent (Variables|Default|GetCollisionAudioComp) self)
  ; End Async Action Watchers
  (Ability|Async|EndAction (Variables|Default|GetTag%20Add%20Watcher))
  (Ability|Async|EndAction (Variables|Default|GetTag%20Remove%20Watcher))
  (Ability|Async|EndAction (Variables|Default|GetSprintTagAddWatcher))
  (Ability|Async|EndAction (Variables|Default|GetSlidingTagAddWatcher))
  (Ability|Async|EndAction (Variables|Default|GetSlidingTagRemoveWatcher)))

(event Custom|Initialize%20Sliding
  (bind _targetplayer (Variables|Default|GetTargetPlayer))
  ;[0 0.05969 0.15 0.5] Sliding on Water
  (bind _asynctaskproxy (Ability|Tasks|WaitGameplayTagAddToActor _targetplayer "(TagName=\"MovementMode.Sliding\")")
    (:then
      (Variables|Default|SetSlidingTagAddWatcher _asynctaskproxy)
      (bind _asynctaskproxy_2 (Ability|Tasks|WaitGameplayTagRemoveFromActor _targetplayer "(TagName=\"MovementMode.Sliding\")")
        (:then
          (Variables|Default|SetSlidingTagRemoveWatcher _asynctaskproxy_2)
          (EventDispatchers|BindEventtoOnSurfaceTypeChanged (Footsteps|GetSurfaceTrackingComponent _targetplayer) (EventDispatchers|CreateEvent :!event Custom|Water%20Check)))
        (:Removed
          (CallFunction|WaterSlide false))))
    (:Added
      (CallFunction|WaterSlide true))))

(event Collision|EventHit
  (params (MyComp :type Primitive%20Component%20Object%20Reference) (Other :type Actor%20Object%20Reference) (OtherComp :type Primitive%20Component%20Object%20Reference) (bSelfMoved :type Boolean) (HitLocation :type Vector :default "0, 0, 0") (HitNormal :type Vector :default "0, 0, 0") (NormalImpulse :type Vector :default "0, 0, 0") (Hit :type Hit%20Result%20Structure%20%28by%20ref%29))
  (Development|PrintString MyComp))

(event Custom|Water%20Check
  (params (SurfaceActor :type Actor%20Object%20Reference) (CurrentSurfaceType :type EPhysicalSurface%20Enum :default "SurfaceType_Default") (PrevSurfaceType :type EPhysicalSurface%20Enum :default "SurfaceType_Default"))
  ;[0 0.4293 1 0.5] Water Check
  (bind _asfort_water_body_actor (Utilities|Casting|CastToFortWaterBodyActor SurfaceActor)
    (:then
      (Variables|Default|SetIsOnWater true)
      (bind _output_get (Variables|Default|SetWater%20Body%20Component (Variables|/Script/Water.WaterBody|GetWaterBodyComponent _asfort_water_body_actor)))
      (bind _footstepwatersplashfx (Variables|Default|GetFootstepWaterSplashFX)
        (:then
          (Water|SetWaterBodyComponent _footstepwatersplashfx "WaterBodyActor" _output_get))
        (:else
          _))
      (bind _watersliding (Variables|Default|GetWaterSliding)
        (:then
          (Water|SetWaterBodyComponent _watersliding "WaterBodyActor" _output_get))
        (:else)))
    (:CastFailed
      (Variables|Default|SetIsOnWater false)
      (Variables|Default|SetWater%20Body%20Component 0))))

(event Custom|On%20Target%20Pawn%20Hit%20Event
  (params (SelfActor :type Actor%20Object%20Reference) (OtherActor :type Actor%20Object%20Reference) (NormalImpulse :type Vector) (Hit :type Hit%20Result%20Structure%20%28by%20ref%29))
  (bind (_bblockinghit _binitialoverlap _time _distance _location _impactpoint _normal _impactnormal) (Collision|BreakHitResult Hit))
  ;[1 1 1] On player hit event, play this audio
  (Audio|OnCollision (Variables|Default|GetCollisionAudioComp) _location (* _impactnormal (Transformation|GetVelocity (Variables|Default|GetTargetPlayer)))))

(event Custom|On%20Player%20Step
  (params (FootIndex :type Integer))
  ;[1 1 1] On player step, play this audio
  (CallFunction|WaterInteraction false)
  (CallFunction|SlidingAudio))

(event Custom|On%20Target%20Pawn%20Landed
  (params (Hit :type Hit%20Result%20Structure%20%28by%20ref%29))
  ;[1 1 1] Landing sound
  (Utilities|FlowControl|DelayUntilNextTick 0)
  (CallFunction|WaterInteraction true)
  (CallFunction|SlidingAudio))

(event Custom|On%20Target%20Pawn%20Movement%20Mode%20Changed
  (params (Character :type Character%20Object%20Reference) (PrevMovementMode :type EMovementMode%20Enum :default "MOVE_None") (PreviousCustomMode :type Byte))
  ;[1 1 1] Sound played when leaving ground
  (Utilities|IsValid Character
    (:Is%20Valid
      (if (not (Variables|/Script/FortniteGame.FortPawn|GetbIsSliding (Variables|Default|GetTargetPlayer)))
        (bind _output_get (Variables|Default|SetMovement%20Mode (Variables|/Script/Engine.CharacterMovementComponent|GetMovementMode (Variables|/Script/Engine.Character|GetCharacterMovement Character))))
        (switch Utilities|FlowControl|Switch|SwitchonEMovementMode _output_get
          (:MOVE_Falling
            (CallFunction|SlidingAudio) ; turn off sliding audio when in air
          ))))))

(event Custom|Sliding%20Audio
  ;[1 1 1] Sliding Audio
  (if (CallFunction|IsSliding)
    ;;[0.055 0.1372 0.15] Slipping audio logic
    (bind _slidingaudiocomponent (Variables|Default|GetSlidingAudioComponent)
      (:then
        (if (not (Audio|Components|Audio|IsPlaying _slidingaudiocomponent))
          (Audio|Parameter|ExecuteTriggerParameter _slidingaudiocomponent "Activate")))
      (:else
        (bind _returnvalue (Audio|SpawnSoundAttached :Sound (Variables|Sound|GetSlide%20Loop%20Sound) :AttachToComponent (Variables|/Script/Engine.Actor|GetRootComponent (Variables|Default|GetTargetPlayer)) :AttachPointName "None" :LocationType "KeepRelativeOffset" :bStopWhenAttachedToDestroyed true :VolumeMultiplier 1.0 :PitchMultiplier 1.0 :StartTime 0.0 :bAutoDestroy true))
        (bind _output_get (Variables|Default|SetSlidingAudioComponent _returnvalue))
        (Audio|Parameter|ExecuteTriggerParameter _output_get "Activate")))
    (else
      (bind _slidingaudiocomponent_2 (Variables|Default|GetSlidingAudioComponent)
        (:then
          (if (Audio|Components|Audio|IsPlaying _slidingaudiocomponent_2)
            (Audio|Parameter|ExecuteTriggerParameter _slidingaudiocomponent_2 "Deactivate")))
        (:else)))))

(fn AddSoundLibrary
  (SoundLibrary|AddLibrary (WorldSubsystems|GetSoundLibrarySubsystem (Variables|Default|GetTargetPlayer)) (Variables|Default|GetAs%20Fort%20Player%20Pawn%20Athena) "/IceSkates/Assets/audio/SL_IceSkates.SL_IceSkates" :!class /Script/SoundLibrary.SoundLibrarySubsystem) ; Add specific IceSkate foley footstep library
)

(fn IsSliding
  (bind _targetplayer (Variables|Default|GetTargetPlayer))
  (if (Pawn|IsLocallyControlled _targetplayer)
    ; LOCALLY CONTROLLED
    (if (or (and (Variables|/Script/FortniteGame.FortPawn|GetbIsSliding (Variables|Default|GetTargetPlayer)) (Math|Boolean|NORBoolean (Utilities|Enum|Equal(Enum) (Variables|/Script/Engine.CharacterMovementComponent|GetMovementMode (Variables|/Script/Engine.Character|GetCharacterMovement (Variables|Default|GetTargetPlayer))) "MOVE_Falling") (Utilities|Enum|Equal(Enum) (Variables|/Script/Engine.CharacterMovementComponent|GetMovementMode (Variables|/Script/Engine.Character|GetCharacterMovement (Variables|Default|GetTargetPlayer))) "MOVE_Flying"))) (and (Utilities|Enum|Equal(Enum) (Variables|/Script/Engine.CharacterMovementComponent|GetMovementMode (Variables|/Script/Engine.Character|GetCharacterMovement (Variables|Default|GetTargetPlayer))) "MOVE_Walking") (== (Math|Vector|VectorLength (+ (Pawn|Components|PawnMovement|GetLastInputVector (Variables|/Script/Engine.Character|GetCharacterMovement (Variables|Default|GetTargetPlayer))) (Pawn|Components|PawnMovement|GetPendingInputVector (Variables|/Script/Engine.Character|GetCharacterMovement (Variables|Default|GetTargetPlayer))))) 0)))
      (return true)
      (else
        (return false)))
    (elif (and (Variables|/Script/FortniteGame.FortPawn|GetbIsSliding _targetplayer) (Math|Boolean|NORBoolean (Utilities|Enum|Equal(Enum) (Variables|/Script/Engine.CharacterMovementComponent|GetMovementMode (Variables|/Script/Engine.Character|GetCharacterMovement _targetplayer)) "MOVE_Falling") (Utilities|Enum|Equal(Enum) (Variables|/Script/Engine.CharacterMovementComponent|GetMovementMode (Variables|/Script/Engine.Character|GetCharacterMovement _targetplayer)) "MOVE_Flying")))
      ; NON-LOCAL PLAYER
      (return true)
      (else
        (return false)))))

(fn RemoveSoundLibrary
  (SoundLibrary|RemoveLibrary (WorldSubsystems|GetSoundLibrarySubsystem (Variables|Default|GetTargetPlayer)) (Variables|Default|GetAs%20Fort%20Player%20Pawn%20Athena) "/IceSkates/Assets/audio/SL_IceSkates.SL_IceSkates" :!class /Script/SoundLibrary.SoundLibrarySubsystem))

(fn ConstructionScript)

(fn WaterSlide
  (params (Active :type Boolean))
  (if (and Active (Variables|Default|GetIsOnWater))
    (bind _watersliding (Variables|Default|GetWaterSliding)
      (:then
        (Components|Activation|Activate _watersliding false))
      (:else
        (bind _returnvalue (Niagara|SpawnSystemAttached "/Game/Effects/Fort_Effects/Gameplay/Sliding/NS_Sliding_Water.NS_Sliding_Water" (Variables|/Script/Engine.Character|GetMesh (Variables|Default|GetTargetPlayer)) "None" "0, 0, 0" "0, 0, 0" "SnapToTarget" true))
        (bind _output_get (Variables|Default|SetWaterSliding _returnvalue))
        (bind _water_body_component (Variables|Default|GetWater%20Body%20Component)
          (:then
            (Water|SetWaterBodyComponent _output_get "WaterBodyActor" _water_body_component)
            (CallFunction|SlidingAudio))
          (:else))))
    (else
      (Components|Activation|Deactivate (Variables|Default|GetWaterSliding)))))

(fn WaterInteraction
  (params (IsLanded :type Boolean))
  (bind _footsteppositionarray (LocalVariables|GetFootStepPositionArray))
  (bind (_output_x _output_y _output_z) (Math|Vector4|BreakVector4 (Utilities|Array|Get(acopy) _footsteppositionarray)))
  (bind _returnvalue (Math|Vector|MakeVector _output_x _output_y _output_z))
  (bind _footstepwatersplashfx (Variables|Default|GetFootstepWaterSplashFX))
  (bind _footstepwatersplashfx_2 (Variables|Default|GetFootstepWaterSplashFX))
  (if (Variables|Default|GetIsOnWater)
    (LocalVariables|SetFootStepPositionArray (Utilities|Array|MakeArray (Math|Vector4|MakeVector4 (.x (.location (Math|Transform|SelectTransform (Transformation|GetSocketTransform (Variables|/Script/Engine.Character|GetMesh (Variables|Default|GetTargetPlayer)) "root" "RTS_World") (Math|Transform|SelectTransform (Transformation|GetSocketTransform (Variables|/Script/Engine.Character|GetMesh (Variables|Default|GetTargetPlayer)) "ball_r" "RTS_World") (Transformation|GetSocketTransform (Variables|/Script/Engine.Character|GetMesh (Variables|Default|GetTargetPlayer)) "ball_l" "RTS_World") (== (AthenaPawn|GetCachedFootstepIndex (Variables|Default|GetTargetPlayer)) 1)) IsLanded))) (.y (.location (Math|Transform|SelectTransform (Transformation|GetSocketTransform (Variables|/Script/Engine.Character|GetMesh (Variables|Default|GetTargetPlayer)) "root" "RTS_World") (Math|Transform|SelectTransform (Transformation|GetSocketTransform (Variables|/Script/Engine.Character|GetMesh (Variables|Default|GetTargetPlayer)) "ball_r" "RTS_World") (Transformation|GetSocketTransform (Variables|/Script/Engine.Character|GetMesh (Variables|Default|GetTargetPlayer)) "ball_l" "RTS_World") (== (AthenaPawn|GetCachedFootstepIndex (Variables|Default|GetTargetPlayer)) 1)) IsLanded))) (.z (.location (Math|Transform|SelectTransform (Transformation|GetSocketTransform (Variables|/Script/Engine.Character|GetMesh (Variables|Default|GetTargetPlayer)) "root" "RTS_World") (Math|Transform|SelectTransform (Transformation|GetSocketTransform (Variables|/Script/Engine.Character|GetMesh (Variables|Default|GetTargetPlayer)) "ball_r" "RTS_World") (Transformation|GetSocketTransform (Variables|/Script/Engine.Character|GetMesh (Variables|Default|GetTargetPlayer)) "ball_l" "RTS_World") (== (AthenaPawn|GetCachedFootstepIndex (Variables|Default|GetTargetPlayer)) 1)) IsLanded))) 1.0)))
    (Variables|Default|GetFootstepWaterSplashFX
      (:then
        (Rendering|SetVisibility _footstepwatersplashfx true true)
        (Transformation|SetWorldLocation _footstepwatersplashfx _returnvalue false true)
        (Components|Activation|Activate _footstepwatersplashfx false))
      (:else
        (bind _returnvalue_2 (Niagara|SpawnSystematLocation :SystemTemplate "/Game/Athena/Items/Traps/FreezeTrap/FX/NS_Ice_Foot_Water.NS_Ice_Foot_Water" :Location _returnvalue :Scale "1.000000,1.000000,1.000000" :bAutoDestroy true :bAutoActivate true :PoolingMethod "AutoRelease" :bPreCullCheck true))
        (Variables|Default|SetFootstepWaterSplashFX _returnvalue_2)))
    (Niagara|SetNiagaraVariable(Bool) _footstepwatersplashfx_2 "IsLanded" (Variables|GetIsLanded))
    (Niagara|NiagaraSetVector4Array _footstepwatersplashfx_2 "SpawnList" (LocalVariables|GetFootStepPositionArray))
    (Utilities|Array|Clear (LocalVariables|GetFootStepPositionArray))))
